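# Physics asset for the medium-build female player character's master skeleton, giving it a simplified per-bone collision volume.
# Each body is one primitive: a sphere for the head, capsules for pelvis, spine_05, arms and legs, and boxes for lower spine, hands and feet.
# Neck, clavicles and fingers have no bodies of their own.
# The constraint list is empty, so no joints link the bodies as a ragdoll.
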PhysicsAsset /Game/Characters/Player/Female/Medium/Master_Skeleton/F_Med_Physics
bodies {
 Spine_01 [{shapeType=Box, extentX=26.4404, extentY=30.075, extentZ=7.7987, center=<3.3882 1.1193 0>, rotation=<rot 0 90 -90>}]
 Spine_02 [{shapeType=Box, extentX=7.6857, extentY=28.1621, extentZ=30.3705, center=<0.9674 0.3194 0.0976>}]
 Spine_03 [{shapeType=Box, extentX=28.3, extentY=30.3, extentZ=10.3, center=<0 0.7653 0>, rotation=<rot 0 90 -90>}]
 calf_l [{shapeType=Capsule, length=26.5, radius=8.5, center=<-19.2811 -1.6546 0.0001>, rotation=<rot -90 151.3934 -151.3934>}]
 calf_r [{shapeType=Capsule, length=26.5, radius=8.5, center=<19.0335 1.6222 0>, rotation=<rot -90 151.3934 -151.3934>}]
 foot_l [{shapeType=Box, extentX=13.0669, extentY=26.6747, extentZ=13.2797, center=<-4.3637 6.2241 -0.1399>}]
 foot_r [{shapeType=Box, extentX=13.0669, extentY=26.6747, extentZ=13.2797, center=<3.8099 -6.5494 -0.1399>}]
 hand_l [{shapeType=Box, extentX=20.0611, extentY=7.9455, extentZ=14.2687, center=<8.7811 -0.9487 -0.3467>}]
 hand_r [{shapeType=Box, extentX=20.0611, extentY=7.9455, extentZ=14.2687, center=<-11.0034 0.392 -0.3467>}]
 head [{shapeType=Sphere, radius=15.15}]
 lowerarm_l [{shapeType=Capsule, length=15, radius=7, center=<8.5884 0 0>, rotation=<rot -90 -41.5864 41.5866>}]
 lowerarm_r [{shapeType=Capsule, length=15, radius=7, center=<-12.5375 -0.0001 0>, rotation=<rot -90 -41.5864 41.5866>}]
 pelvis [{shapeType=Capsule, length=15, radius=15}]
 spine_04 [{shapeType=Box, extentX=33.0189, extentY=30.856, extentZ=15.1641, center=<-0.8011 0.8173 0>, rotation=<rot 0 90 -90>}]
 spine_05 [{shapeType=Capsule, length=15, radius=15}]
 thigh_l [{shapeType=Capsule, length=15, radius=9.25, center=<-17.5194 2.1817 -1.0219>, rotation=<rot 89.9011 -5.4268 -5.4268>}]
 thigh_r [{shapeType=Capsule, length=15, radius=9.25, center=<18.0067 -2.7164 0.6907>, rotation=<rot 89.9011 -5.4268 -5.4268>}]
 upperarm_l [{shapeType=Capsule, length=18.5, radius=7.25, center=<10.3085 0.5833 0.4521>, rotation=<rot -89.9161 -38.4742 38.4742>}]
 upperarm_r [{shapeType=Capsule, length=18.5, radius=7.25, center=<-12.6133 0.5831 -1.776>, rotation=<rot -89.9161 -38.4742 38.4742>}]
}
constraints []
